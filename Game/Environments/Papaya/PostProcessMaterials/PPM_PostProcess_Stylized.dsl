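; Blueprint controller for the Papaya environment's stylized outline post-process, exposing events to toggle its PostProcess component and tune the look.
; Individual events write outline color, desaturation, hashmark intensity and line settings to the MPC_PapayaGadgetControls material parameter collection.
; Five preset events build a stylized-settings struct and pass it to SetPostProcessValues, which copies line thickness and two sine-wave parameter sets into that collection.
; The per-parameter line and sine events are labeled as unused detail settings.
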(event Custom|SetPostEnabled
  (params (Enabled :type Boolean))
  ;[1 1 1] Enable Disable
  (Variables|/Script/Engine.PostProcessComponent|SetbEnabled Enabled (Variables|Default|GetPostProcess)))

(event Custom|SetOutlineColor
  (params (Enabled :type Vector))
  ;[1 1 1] Color Settings
  (Rendering|Material|SetVectorParameterValue "/Game/Environments/Papaya/Materials/MPC_PapayaGadgetControls.MPC_PapayaGadgetControls" "OutlineColor" (Math|Conversions|ToLinearColor(Vector) Enabled) :!class /Script/Engine.KismetMaterialLibrary))

(event Custom|SetSceneDesaturation
  (params (SceneDesaturation :type Float%20%28double-precision%29))
  ;[1 1 1] Color Settings
  (Rendering|Material|SetScalarParameterValue "/Game/Environments/Papaya/Materials/MPC_PapayaGadgetControls.MPC_PapayaGadgetControls" "SceneDesaturation" SceneDesaturation :!class /Script/Engine.KismetMaterialLibrary))

(event Custom|SetHasmarkIntensity
  (params (Value :type Float%20%28double-precision%29))
  ;[1 1 1] Color Settings
  (Rendering|Material|SetScalarParameterValue "/Game/Environments/Papaya/Materials/MPC_PapayaGadgetControls.MPC_PapayaGadgetControls" "HasmarksIntensity" Value :!class /Script/Engine.KismetMaterialLibrary))

(event Custom|SetOutlineColorDynamic
  (params (SetOutlineColorDynamic :type Float%20%28double-precision%29))
  ;[1 1 1] Color Settings
  (Rendering|Material|SetScalarParameterValue "/Game/Environments/Papaya/Materials/MPC_PapayaGadgetControls.MPC_PapayaGadgetControls" "OutlineColorDynamic" SetOutlineColorDynamic :!class /Script/Engine.KismetMaterialLibrary))

(event Custom|SetPost_Preset_1
  ;[1 1 1] Presets
  (CallFunction|SetPostProcessValues (Utilities|Struct|MakePPMPostProccesStylizedStruct 0.0 0.0 3.0 3.0)))

(event Custom|SetLineThickness_MinClamp_Normalized
  (params (Value :type Float%20%28double-precision%29))
  ;[1 1 1] Detailed setting not used
  (Rendering|Material|SetScalarParameterValue "/Game/Environments/Papaya/Materials/MPC_PapayaGadgetControls.MPC_PapayaGadgetControls" "LineThickness_MinClamp_Normalized" Value :!class /Script/Engine.KismetMaterialLibrary))

(event Custom|SetLineThicknessAmplitude_Normalized
  (params (Value :type Float%20%28double-precision%29))
  ;[1 1 1] Detailed setting not used
  (Rendering|Material|SetScalarParameterValue "/Game/Environments/Papaya/Materials/MPC_PapayaGadgetControls.MPC_PapayaGadgetControls" "LineThicknessAmplitude_Normalized" Value :!class /Script/Engine.KismetMaterialLibrary))

(event Custom|SetPost_Preset_2
  ;[1 1 1] Presets
  (CallFunction|SetPostProcessValues (Utilities|Struct|MakePPMPostProccesStylizedStruct 0.0 0.0 8.0 8.0)))

(event Custom|SetLineThickness_Min
  (params (Value :type Float%20%28double-precision%29))
  ;[1 1 1] Detailed setting not used
  (Rendering|Material|SetScalarParameterValue "/Game/Environments/Papaya/Materials/MPC_PapayaGadgetControls.MPC_PapayaGadgetControls" "LineThickness_Min" Value :!class /Script/Engine.KismetMaterialLibrary))

(event Custom|SetLineThickness_Max
  (params (Value :type Float%20%28double-precision%29))
  ;[1 1 1] Detailed setting not used
  (Rendering|Material|SetScalarParameterValue "/Game/Environments/Papaya/Materials/MPC_PapayaGadgetControls.MPC_PapayaGadgetControls" "LineThickness_Max" Value :!class /Script/Engine.KismetMaterialLibrary))

(event Custom|SetPost_Preset_3
  ;[1 1 1] Presets
  (CallFunction|SetPostProcessValues (Utilities|Struct|MakePPMPostProccesStylizedStruct 0.05 0.1 0.0 1e+01 1.0 0.8 5.0 0.9 0.3 12.0)))

(event Custom|SetAmplitude_Normalized1
  (params (Value :type Float%20%28double-precision%29))
  ;[1 1 1] Detailed setting not used
  (Rendering|Material|SetScalarParameterValue "/Game/Environments/Papaya/Materials/MPC_PapayaGadgetControls.MPC_PapayaGadgetControls" "SineAmplitude_Normalized1" Value :!class /Script/Engine.KismetMaterialLibrary))

(event Custom|SetSineSpeed1
  (params (Value :type Float%20%28double-precision%29))
  ;[1 1 1] Detailed setting not used
  (Rendering|Material|SetScalarParameterValue "/Game/Environments/Papaya/Materials/MPC_PapayaGadgetControls.MPC_PapayaGadgetControls" "SineSpeed1" Value :!class /Script/Engine.KismetMaterialLibrary))

(event Custom|SetSineFrequency1
  (params (Value :type Float%20%28double-precision%29))
  ;[1 1 1] Detailed setting not used
  (Rendering|Material|SetScalarParameterValue "/Game/Environments/Papaya/Materials/MPC_PapayaGadgetControls.MPC_PapayaGadgetControls" "SineFrequency1" Value :!class /Script/Engine.KismetMaterialLibrary))

(event Custom|SetPost_Preset_4
  ;[1 1 1] Presets
  (CallFunction|SetPostProcessValues (Utilities|Struct|MakePPMPostProccesStylizedStruct 0.05 0.1 0.0 1e+01 1.0 2.0 5.0 0.9 1.5 12.0)))

(event Custom|SetAmplitude_Normalized2
  (params (Value :type Float%20%28double-precision%29))
  ;[1 1 1] Detailed setting not used
  (Rendering|Material|SetScalarParameterValue "/Game/Environments/Papaya/Materials/MPC_PapayaGadgetControls.MPC_PapayaGadgetControls" "SineAmplitude_Normalized2" Value :!class /Script/Engine.KismetMaterialLibrary))

(event Custom|SetSineSpeed2
  (params (Value :type Float%20%28double-precision%29))
  ;[1 1 1] Detailed setting not used
  (Rendering|Material|SetScalarParameterValue "/Game/Environments/Papaya/Materials/MPC_PapayaGadgetControls.MPC_PapayaGadgetControls" "SineSpeed2" Value :!class /Script/Engine.KismetMaterialLibrary))

(event Custom|SetSineFrequency2
  (params (Value :type Float%20%28double-precision%29))
  ;[1 1 1] Detailed setting not used
  (Rendering|Material|SetScalarParameterValue "/Game/Environments/Papaya/Materials/MPC_PapayaGadgetControls.MPC_PapayaGadgetControls" "SineFrequency2" Value :!class /Script/Engine.KismetMaterialLibrary))

(event Custom|SetPost_Preset_5
  ;[1 1 1] Presets
  (CallFunction|SetPostProcessValues (Utilities|Struct|MakePPMPostProccesStylizedStruct 0.05 1.0 0.0 2e+01)))

(fn SetPostProcessValues
  (params (NewParam :type PPM%20Post%20Procces%20Stylized%20Struct%20Structure))
  (bind (_linethickness_minclamp_normalized_8_c80ccfb942942b84f10620904a902569 _linethicknessamplitude_normalized_3_8185961f4605d9ffc70c01b772fbcfe6 _linethickness_min_9_1c41ef324bfd364e3c17769a524ce8e4 _linethickness_max_10_4090b03a44b346ab1ef7749023218c0b _sineamplitude_normalized1_13_5a63418c4eaa46c313dae2b0dfad6f48 _sinespeed1_14_1d3706044fd8280b053875b0e3d73763 _sinefrequency1_16_5121ea5b43c85eb006e9a3b783d8808f _sineamplitude_normalized2_20_6a73115a422f52b746fc2c8288da4154 _sinespeed2_21_1a568492401943ef14465083bffbca0e _sinefrequency2_22_df080c904b5b0a5c886aab99c08acf6d) (Utilities|Struct|BreakPPMPostProccesStylizedStruct NewParam))
  (Rendering|Material|SetScalarParameterValue "/Game/Environments/Papaya/Materials/MPC_PapayaGadgetControls.MPC_PapayaGadgetControls" "LineThickness_MinClamp_Normalized" _linethickness_minclamp_normalized_8_c80ccfb942942b84f10620904a902569 :!class /Script/Engine.KismetMaterialLibrary)
  (Rendering|Material|SetScalarParameterValue "/Game/Environments/Papaya/Materials/MPC_PapayaGadgetControls.MPC_PapayaGadgetControls" "LineThicknessAmplitude_Normalized" _linethicknessamplitude_normalized_3_8185961f4605d9ffc70c01b772fbcfe6 :!class /Script/Engine.KismetMaterialLibrary)
  (Rendering|Material|SetScalarParameterValue "/Game/Environments/Papaya/Materials/MPC_PapayaGadgetControls.MPC_PapayaGadgetControls" "LineThickness_Min" _linethickness_min_9_1c41ef324bfd364e3c17769a524ce8e4 :!class /Script/Engine.KismetMaterialLibrary)
  (Rendering|Material|SetScalarParameterValue "/Game/Environments/Papaya/Materials/MPC_PapayaGadgetControls.MPC_PapayaGadgetControls" "LineThickness_Max" _linethickness_max_10_4090b03a44b346ab1ef7749023218c0b :!class /Script/Engine.KismetMaterialLibrary)
  (Rendering|Material|SetScalarParameterValue "/Game/Environments/Papaya/Materials/MPC_PapayaGadgetControls.MPC_PapayaGadgetControls" "SineAmplitude_Normalized1" _sineamplitude_normalized1_13_5a63418c4eaa46c313dae2b0dfad6f48 :!class /Script/Engine.KismetMaterialLibrary)
  (Rendering|Material|SetScalarParameterValue "/Game/Environments/Papaya/Materials/MPC_PapayaGadgetControls.MPC_PapayaGadgetControls" "SineSpeed1" _sinespeed1_14_1d3706044fd8280b053875b0e3d73763 :!class /Script/Engine.KismetMaterialLibrary)
  (Rendering|Material|SetScalarParameterValue "/Game/Environments/Papaya/Materials/MPC_PapayaGadgetControls.MPC_PapayaGadgetControls" "SineFrequency1" _sinefrequency1_16_5121ea5b43c85eb006e9a3b783d8808f :!class /Script/Engine.KismetMaterialLibrary)
  (Rendering|Material|SetScalarParameterValue "/Game/Environments/Papaya/Materials/MPC_PapayaGadgetControls.MPC_PapayaGadgetControls" "SineAmplitude_Normalized2" _sineamplitude_normalized2_20_6a73115a422f52b746fc2c8288da4154 :!class /Script/Engine.KismetMaterialLibrary)
  (Rendering|Material|SetScalarParameterValue "/Game/Environments/Papaya/Materials/MPC_PapayaGadgetControls.MPC_PapayaGadgetControls" "SineSpeed2" _sinespeed2_21_1a568492401943ef14465083bffbca0e :!class /Script/Engine.KismetMaterialLibrary)
  (Rendering|Material|SetScalarParameterValue "/Game/Environments/Papaya/Materials/MPC_PapayaGadgetControls.MPC_PapayaGadgetControls" "SineFrequency2" _sinefrequency2_22_df080c904b5b0a5c886aab99c08acf6d :!class /Script/Engine.KismetMaterialLibrary))

(fn ConstructionScript)
