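# Papaya post-process material that draws audio-reactive sketch outlines and hatching over an optionally desaturated scene.
# Four offset depth samples form distance-faded edges, with thickness driven by event audio and animated screen-space sines.
# Deferred adds two-scale world-aligned hatching in distance-field-occluded areas, masked off visible custom-depth objects; forward and mobile keep outlines only.
# Line color blends static and gadget-driven colors, with controls switchable between MPC_PapayaGadgetControls and local parameters.
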Material /Game/Environments/Papaya/PostProcessMaterials/PPM_StylizedSketchOutline
graph {
 properties {
  bEnableResponsiveAA=true
  blendMode=BLEND_Translucent
  blendableLocation=BL_SceneColorBeforeDOF
  blendablePriority=-5
  materialDomain=MD_PostProcess
  pixelDepthOffsetMode=PDOM_Legacy
  shadingModel=MSM_Unlit
  usedShadingModels=Unlit
 }
 nodes [
  n1 = CollectionParameter() {collection=&FortniteMaterialParameters, parameterName=EventAudio_Amplitude}
  n2 = Multiply(
   A=@n1
   B=StaticSwitchParameter(
    True=CollectionParameter() {collection=/Game/Environments/Papaya/Materials/MPC_PapayaGadgetControls, parameterName=LineThicknessAmplitude_Normalized}
    False=ScalarParameter() {group=LineThickness, parameterName=LineThicknessAmplitude_Normalized}
   ) {defaultValue=true, parameterName=UseMCP}
  )
  n3 = CollectionParameter() {collection=&FortniteMaterialParameters, parameterName=EventAudio_Spectrum}
  n4 = MaterialFunctionCall(Float4=@n3) {materialFunction=&BreakOutFloat4Components}
  n5 = ScreenPosition()
  n6 = Multiply(
   A=Add(A=ComponentMask(None=@n5.ViewportUV) {r=true}, B=ComponentMask(None=@n5.ViewportUV) {g=true})
   B=StaticSwitchParameter(
    True=CollectionParameter() {collection=&MPC_PapayaGadgetControls, parameterName=SineFrequency1}
    False=ScalarParameter() {defaultValue=1, group=LineThicknessSine1, parameterName=SineFrequency1}
   ) {defaultValue=true, parameterName=UseMCP}
  )
  n7 = Multiply(
   A=Time()
   B=StaticSwitchParameter(
    True=CollectionParameter() {collection=&MPC_PapayaGadgetControls, parameterName=SineSpeed1}
    False=ScalarParameter() {defaultValue=1, group=LineThicknessSine1, parameterName=SineSpeed1}
   ) {defaultValue=true, parameterName=UseMCP}
  )
  n8 = Add(A=Sine(None=Add(A=@n6, B=@n7)))
  n9 = CollectionParameter() {collection=&MPC_PapayaGadgetControls, parameterName=SineAmplitude_Normalized1}
  n10 = ScalarParameter() {defaultValue=1, group=LineThicknessSine1, parameterName=SineAmplitude_Normalized1}
  n11 = Multiply(A=@n4.G, B=Multiply(A=Divide(A=@n8), B=StaticSwitchParameter(True=@n9, False=@n10) {defaultValue=true, parameterName=UseMCP}))
  Add(A=@n2, B=@n11)
  n12 = ViewSize()
  n13 = ScreenPosition()
  n14 = Add(A=Multiply(A=ComponentMask(None=@n13.ViewportUV) {r=true}) {constB=-1}, B=ComponentMask(None=@n13.ViewportUV) {g=true})
  n15 = StaticSwitchParameter(
   True=CollectionParameter() {collection=&MPC_PapayaGadgetControls, parameterName=SineFrequency2}
   False=ScalarParameter() {defaultValue=1, group=LineThicknessSine2, parameterName=SineFrequency2}
  ) {defaultValue=true, parameterName=UseMCP}
  n16 = Time()
  n17 = StaticSwitchParameter(
   True=CollectionParameter() {collection=&MPC_PapayaGadgetControls, parameterName=SineSpeed2}
   False=ScalarParameter() {defaultValue=1, group=LineThicknessSine2, parameterName=SineSpeed2}
  ) {defaultValue=true, parameterName=UseMCP}
  n18 = Sine(None=Add(A=Multiply(A=@n14, B=@n15), B=Multiply(A=@n16, B=@n17)))
  n19 = Multiply(
   A=Divide(A=Add(A=@n18))
   B=StaticSwitchParameter(
    True=CollectionParameter() {collection=&MPC_PapayaGadgetControls, parameterName=SineAmplitude_Normalized2}
    False=ScalarParameter() {defaultValue=1, group=LineThicknessSine2, parameterName=SineAmplitude_Normalized2}
   ) {defaultValue=true, parameterName=UseMCP}
  )
  n20 = Max(
   A=StaticSwitchParameter(
    True=CollectionParameter() {collection=&MPC_PapayaGadgetControls, parameterName=LineThickness_MinClamp_Normalized}
    False=ScalarParameter() {group=LineThickness, parameterName=LineThickness_MinClamp_Normalized}
   ) {defaultValue=true, parameterName=UseMCP}
   B=Max(A=Max(A=@n2, B=@n11), B=Multiply(A=@n4.R, B=@n19))
  )
  n21 = Divide(
   A=@n12
   B=LinearInterpolate(
    A=ScalarParameter() {defaultValue=50, group=LineThickness, parameterName=LineThickness_Divide_Min}
    B=ScalarParameter() {defaultValue=2000, group=LineThickness, parameterName=LineThickness_Divide_Max}
    Alpha=@n20
   )
  )
  Clamp(None=Ceil(None=@n21)) {maxDefault=8}
  ComponentMask(None=SceneTexture() {sceneTextureId=PPI_AmbientOcclusion}.Color) {r=true}
  Constant() {r=-8}
  Constant() {r=2}
  n22 = MaterialFunctionCall(
   TextureObject=TextureObject() {samplerType=SAMPLERTYPE_Masks, texture=/Game/Global/Textures/Default/Tileable_Patterns/OakPP_Post}
   TextureSize=ScalarParameter() {defaultValue=800, parameterName=Size}
   World Space Normal=ComponentMask(None=SceneTexture() {sceneTextureId=PPI_WorldNormal}.Color) {b=true, g=true, r=true}
  ) {materialFunction=&WorldAlignedTexture}
  n23 = TextureObject() {samplerType=SAMPLERTYPE_Masks, texture=&OakPP_Post}
  n24 = ScalarParameter() {defaultValue=8000, parameterName=Size2}
  n25 = ComponentMask(None=SceneTexture() {sceneTextureId=PPI_WorldNormal}.Color) {b=true, g=true, r=true}
  n26 = Divide(
   A=Add(A=SceneDepth().R, B=ScalarParameter() {defaultValue=-3000, parameterName=DistanceOffset})
   B=ScalarParameter() {defaultValue=5000, parameterName=Distance}
  )
  n27 = ComponentMask(None=LinearInterpolate(
    A=@n22.XYZ Texture
    B=MaterialFunctionCall(TextureObject=@n23, TextureSize=@n24, World Space Normal=@n25) {materialFunction=&WorldAlignedTexture}.XYZ Texture
    Alpha=Saturate(None=@n26)
   )) {r=true}
  n28 = ScalarParameter() {defaultValue=200, parameterName=Dist}
  n29 = Add(A=WorldPosition().XYZ, B=Multiply(A=Normalize(VectorInput=@n25), B=@n28))
  n30 = Saturate(None=Divide(A=DistanceToNearestSurface!(World Position=@n29), B=@n28))
  n31 = Saturate(None=LinearInterpolate(A=ScalarParameter() {defaultValue=-20, parameterName=AOC2}, B=ScalarParameter() {defaultValue=1.125, parameterName=AOC1}, Alpha=@n30))
  n32 = CollectionParameter() {collection=&MPC_PapayaGadgetControls, parameterName=HasmarksIntensity}
  n33 = ScalarParameter() {defaultValue=1, group=Color, parameterName=Highquality_Hasmarks_Intensity}
  n34 = LinearInterpolate(
   A=Constant() {r=1}
   B=Multiply(A=ScalarParameter() {defaultValue=0.9, parameterName=DarkenNoise}, B=@n27)
   Alpha=Multiply(A=OneMinus(None=@n31), B=StaticSwitchParameter(True=@n32, False=@n33) {defaultValue=true, parameterName=UseMCP})
  )
  n35 = Divide(
   A=ComponentMask(None=SceneTexture() {sceneTextureId=PPI_CustomDepth}.Color) {r=true}
   B=ScalarParameter() {defaultValue=50000000, parameterName=MaskCharacterDistance}
  )
  LinearInterpolate(A=Constant() {r=1}, B=@n34, Alpha=Saturate(None=Floor(None=@n35)))
  Multiply(A=@n21, B=SceneTexelSize()) {constA=2}
  n36 = ComponentMask(None=@n22.XYZ Texture) {r=true}
  QualitySwitch(Default=@n36, Low=@n36, High=@n27, Medium=@n27)
  n37 = ComponentMask(None=SceneTexture() {sceneTextureId=PPI_PostProcessInput0}.Color) {b=true, g=true, r=true}
  n38 = StaticSwitchParameter(
   True=CollectionParameter() {collection=&MPC_PapayaGadgetControls, parameterName=SceneDesaturation}
   False=ScalarParameter() {group=Color, parameterName=SceneDesaturation}
  ) {defaultValue=true, parameterName=UseMCP}
  n39 = SceneTexture() {sceneTextureId=PPI_SceneDepth}
  n40 = ComponentMask(None=ScreenPosition().ViewportUV) {g=true, r=true}
  n41 = StaticSwitchParameter(
   True=CollectionParameter() {collection=&MPC_PapayaGadgetControls, parameterName=LineThickness_Min}
   False=ScalarParameter() {defaultValue=1, group=LineThickness, parameterName=LineThickness_Min}
  ) {defaultValue=true, parameterName=UseMCP}
  n42 = StaticSwitchParameter(
   True=CollectionParameter() {collection=&MPC_PapayaGadgetControls, parameterName=LineThickness_Max}
   False=ScalarParameter() {defaultValue=3, group=LineThickness, parameterName=LineThickness_Max}
  ) {defaultValue=true, parameterName=UseMCP}
  n43 = ScalarParameter() {defaultValue=2000, parameterName=BiasLineThickness}
  n44 = Multiply(A=Multiply(A=LinearInterpolate(A=@n41, B=@n42, Alpha=@n20), B=Divide(A=@n12, B=@n43)), B=SceneTexelSize())
  n45 = Constant()
  n46 = AppendVector(A=ComponentMask(None=@n44) {desc=x, r=true}, B=@n45)
  n47 = Add(A=@n40, B=@n46) {desc=x+}
  n48 = AppendVector(A=@n45, B=ComponentMask(None=@n44) {desc=y, g=true}) {desc=y+}
  n49 = Add(A=@n40, B=@n48) {desc=y+}
  n50 = Constant() {r=-1}
  n51 = Add(A=@n40, B=Multiply(A=@n46, B=@n50)) {desc=x-}
  n52 = Add(A=@n40, B=Multiply(A=@n50, B=@n48)) {desc=y-}
  n53 = ScalarParameter() {defaultValue=-12.739, parameterName=Bias_Near}
  n54 = ScalarParameter() {defaultValue=-800, parameterName=Bias_Far}
  n55 = ComponentMask(None=@n39.Color) {r=true}
  n56 = ScalarParameter() {defaultValue=1000, parameterName=FadeBiasOffset}
  n57 = ScalarParameter() {defaultValue=20000, parameterName=FadeDist}
  n58 = Saturate(None=Divide(A=Subtract(A=@n55, B=@n56), B=@n57))
  n59 = ComponentMask(None=MaterialFunctionCall!(
    Base=@n39.Color
    BaseSampleBias=ScalarParameter() {defaultValue=1, parameterName=BaseSampleBias}
    Scale
    Sample1=SceneTexture(UVs=@n47) {sceneTextureId=PPI_SceneDepth}.Color
    Sample2=SceneTexture(UVs=@n49) {sceneTextureId=PPI_SceneDepth}.Color
    Sample3=SceneTexture(UVs=@n51) {sceneTextureId=PPI_SceneDepth}.Color
    Sample4=SceneTexture(UVs=@n52) {sceneTextureId=PPI_SceneDepth}.Color
    Divisor=LinearInterpolate(A=@n53, B=@n54, Alpha=@n58)
    UseCenteredLines
    UseRGB
   ) {materialFunction=/Game/Packages/jl_tests/combine4Samples}.Result) {r=true}
  n60 = Subtract(A=@n55, B=@n57)
  n61 = ScalarParameter() {defaultValue=20000, parameterName=FadeRadius}
  n62 = OneMinus(None=Clamp(None=Divide(A=@n60, B=@n61)))
  n63 = OneMinus(None=ComponentMask(None=Multiply(A=@n59, B=@n62)) {r=true})
  n64 = SceneTexture() {bFiltered=true, sceneTextureId=PPI_SceneDepth}
  n65 = Subtract(
   A=Add(A=ScalarParameter() {defaultValue=5, parameterName=Test}, B=ComponentMask(None=@n64.Color) {r=true}) {constB=0.5}
   B=ComponentMask(None=SceneTexture() {bFiltered=true, sceneTextureId=PPI_CustomDepth}.Color) {r=true}
  )
  n66 = Multiply(A=@n63, B=LinearInterpolate(A=@n34, B=Constant() {r=1}, Alpha=Clamp(None=@n65)))
  n67 = StaticSwitchParameter(
   True=ComponentMask(None=CollectionParameter() {collection=&MPC_PapayaGadgetControls, parameterName=OutlineColor}) {b=true, g=true, r=true}
   False=VectorParameter() {group=Color, parameterName=OutlineColor, defaultValue=<rgba 0 0 0 1>}.RGB
  ) {defaultValue=true, parameterName=UseMCP}
  n68 = Desaturation(
   None=Multiply(
    A=MaterialFunctionCall!(
     Media Player TexCoord
     Media Color Code Coordinate
     Media Texture Resolution
     Base Color Parameter
     Base Color Mix
     Media Color Code Mix
     Intermission Color Parameter
     Intermission Color Mix
     Desaturation
    ) {materialFunction=/Game/Environments/Papaya/Materials/MaterialFunctions/MF_Papaya_GadgetColor}.Result
    B=ScalarParameter() {defaultValue=2, parameterName=Color Multiplier}
   )
   Fraction=ScalarParameter() {defaultValue=0.125, parameterName=Color Desat}
  )
  n69 = StaticSwitchParameter(
   True=CollectionParameter() {collection=&MPC_PapayaGadgetControls, parameterName=OutlineColorDynamic}
   False=ScalarParameter() {group=Color, parameterName=OutlineColorDynamic}
  ) {defaultValue=true, parameterName=UseMCP}
  Reroute(None=LinearInterpolate(
    A=LinearInterpolate(A=@n67, B=@n68, Alpha=@n69)
    B=Desaturation(None=@n37, Fraction=@n38)
    Alpha=ShadingPathSwitch(Default=@n66, Deferred=@n66, Forward=@n63, Mobile=@n63)
   ) {constB=0})
  Reroute(None=@n30)
 ]
}
